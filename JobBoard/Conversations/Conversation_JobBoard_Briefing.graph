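# Conversation database for job board briefings, with entry points for treasure hunt, supply drop, vehicle drop, bounty and weapon specialist, plus a "no jobs available" fallback.
# Each job's speech node offers cancel and grant choices, gated by inverted active-quest requirements.
# All entries share three "already active" refusal choices shown by matching tags.
# Grants run side effects that record use, then variously spawn actors or loot, apply effects or deactivate the board.
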ConversationDatabase /JobBoard/Conversations/Conversation_JobBoard_Briefing
entry_points [
 {tag=Conversation.Entry.JobBoard, destinations [#1]}
 {tag=Conversation.Entry.JobBoard.TreasureHunt, destinations [#2]}
 {tag=Conversation.Entry.JobBoard.SupplyDropGeneric, destinations [#3]}
 {tag=Conversation.Entry.JobBoard.VehicleDrop, destinations [#4]}
 {tag=Conversation.Entry.JobBoard.Bounty, destinations [#5]}
 {tag=Conversation.Entry.JobBoard.WeaponSpecialist, destinations [#6]}
]
nodes [
 ConversationEntryPointNode_5 {
  id=#6
  class=ConversationEntryPointNode
  outputs [#7]
  properties {entryTag {tagName=Conversation.Entry.JobBoard.WeaponSpecialist}}
 }
 Speech_C_12 {
  id=#7
  class=Speech_C
  outputs [
   #8
   #9
   #10
   #11
   #12
  ]
  properties {nodeName=""}
 }
 GrantStaticQuest_C_3 {
  id=#12
  class=GrantStaticQuest_C
  outputs []
  properties {
   bSharedQuestLimitsEnforced=false
   bShouldOverrideSpeakerText=false
   briefWidgetClassOverride=/JobBoard/Widgets/FortWeaponJobBrief.FortWeaponJobBrief_C
   questToGrantPtr=/JobBoard/QuestItems/WeaponSpecialist/quest_evergreen_jobboard_armory_allweapons
   softTaskIcon=/Game/UI/Foundation/Textures/Icons/Checks/T-Icon-Check-128
  }
  sub_nodes [
   FortConversationRequirement_ControllerRequirement_79 {
    class=FortConversationRequirement_ControllerRequirement
    properties {
     nodeName=Require: Has Urgent Quest
     requirement {
      class=FortControllerRequirement_HasUrgentQuest
      properties {
       bInverseCondition=true
       bRequireExactMatch=false
       acceptedUrgentEventTags {gameplayTags [{tagName=SpecialEvent.HuntingPlayer.IsHunter}, {tagName=SpecialEvent.JobBoard.Experimental}]}
      }
     }
    }
   }
   FortConversationRequirement_ControllerRequirement_80 {
    class=FortConversationRequirement_ControllerRequirement
    properties {
     nodeName=Require: Has GameplayTag
     requirement {
      class=FortControllerRequirement_HasOwnedTag
      properties {bInverseCondition=true, ownedTag {tagName=Granted.Athena.JobBoardExperimental.QuestActive}}
     }
    }
   }
   FortConversationRequirement_ControllerRequirement_81 {
    class=FortConversationRequirement_ControllerRequirement
    properties {
     nodeName=Require: Has GameplayTag
     requirement {class=FortControllerRequirement_HasOwnedTag, properties {bInverseCondition=true, ownedTag {tagName=Granted.Athena.BountyBoard.QuestActive}}}
    }
   }
   FortConversationRequirement_ControllerRequirement_82 {
    class=FortConversationRequirement_ControllerRequirement
    properties {
     nodeName=Require: Has GameplayTag
     requirement {class=FortControllerRequirement_HasOwnedTag, properties {bInverseCondition=true, ownedTag {tagName=Granted.Athena.UrgentQuest.Active}}}
    }
   }
   Requirement_JobBoardBriefing_CanGiveLootForTier_C_4 {class=Requirement_JobBoardBriefing_CanGiveLootForTier_C, properties {lootTierGroupName=LTG_ArmoryBriefing}}
   ConversationChoiceNode_31 {class=ConversationChoiceNode, properties {defaultChoiceDisplayText=HUNTER BRIEFING}}
   SideEffect_PlayServiceUsedEffects_C_16 {class=SideEffect_PlayServiceUsedEffects_C}
   SideEffect_Server_RecordEventToHistoryData_C_16 {class=SideEffect_Server_RecordEventToHistoryData_C, properties {event Type {tagName=Conversation.Event.GrantedQuest}}}
   SideEffect_Server_TrackJobBoardUser_C_16 {class=SideEffect_Server_TrackJobBoardUser_C}
   SideEffect_Server_JobBoardBriefing_SpawnLoot_C_5 {class=SideEffect_Server_JobBoardBriefing_SpawnLoot_C, properties {lootTierGroupName=LTG_ArmoryBriefing}}
  ]
 }
 Continue_C_6 {
  id=#11
  class=Continue_C
  properties {softTaskIcon=&Cancel_Normal}
  sub_nodes [
   ConversationChoiceNode_19 {
    class=ConversationChoiceNode
    properties {defaultChoiceDisplayText=CANCEL, choiceTags {gameplayTags [{tagName=Conversation.Choice.CustomSortPriority.Highest}]}}
   }
   FortConversationRequirement_ControllerRequirement_20 {
    class=FortConversationRequirement_ControllerRequirement
    properties {
     nodeName=Require: Has GameplayTag
     requirement {class=FortControllerRequirement_HasOwnedTag, properties {bInverseCondition=true, ownedTag {tagName=Granted.Athena.BountyBoard.QuestActive}}}
    }
   }
   FortConversationRequirement_ControllerRequirement_25 {
    class=FortConversationRequirement_ControllerRequirement
    properties {
     nodeName=Require: Has GameplayTag
     requirement {
      class=FortControllerRequirement_HasOwnedTag
      properties {bInverseCondition=true, ownedTag {tagName=Granted.Athena.JobBoardExperimental.QuestActive}}
     }
    }
   }
   FortConversationRequirement_ControllerRequirement_32 {
    class=FortConversationRequirement_ControllerRequirement
    properties {
     nodeName=Require: Has GameplayTag
     requirement {class=FortControllerRequirement_HasOwnedTag, properties {bInverseCondition=true, ownedTag {tagName=Granted.Athena.UrgentQuest.Active}}}
    }
   }
  ]
 }
 ShowUserDataDrivenCannotUseReason_C_5 {
  id=#10
  class=ShowUserDataDrivenCannotUseReason_C
  outputs [#13]
  properties {
   cannotUseReason=UrgentQuestAlreadyActive
   hoverChoiceSound=/JobBoard/Sounds/JobBoard_ShadowBriefing_NotAvailable_Cue
   serviceBriefCollectionKeyOveride=Service.JobBoard.DisplayCannotUseReason
   serviceBriefConfigCollection=&/JobBoardServiceBriefCollectionConfig
   softTaskIcon=&Cancel_Normal
  }
  sub_nodes [
   ConversationChoiceNode_22 {
    class=ConversationChoiceNode
    properties {defaultChoiceDisplayText=URGENT QUEST ALREADY ACTIVE, choiceTags {gameplayTags [{tagName=Conversation.Choice.Category.Back}]}}
   }
   FortConversationRequirement_ControllerRequirement_44 {
    class=FortConversationRequirement_ControllerRequirement
    properties {
     nodeName=Require: Has GameplayTag
     requirement {class=FortControllerRequirement_HasOwnedTag, properties {ownedTag {tagName=Granted.Athena.UrgentQuest.Active}}}
    }
   }
  ]
 }
 Continue_C_8 {id=#13, class=Continue_C}
 ShowUserDataDrivenCannotUseReason_C_0 {
  id=#9
  class=ShowUserDataDrivenCannotUseReason_C
  outputs [#14]
  properties {
   cannotUseReason=BriefingAlreadyActive
   hoverChoiceSound=&JobBoard_ShadowBriefing_NotAvailable_Cue
   serviceBriefCollectionKeyOveride=Service.JobBoard.DisplayCannotUseReason
   serviceBriefConfigCollection=&JobBoardServiceBriefCollectionConfig
   softTaskIcon=&Cancel_Normal
  }
  sub_nodes [
   ConversationChoiceNode_16 {
    class=ConversationChoiceNode
    properties {defaultChoiceDisplayText=JOB ALREADY ACTIVE, choiceTags {gameplayTags [{tagName=Conversation.Choice.Category.Back}]}}
   }
   FortConversationRequirement_ControllerRequirement_26 {
    class=FortConversationRequirement_ControllerRequirement
    properties {
     nodeName=Require: Has GameplayTag
     requirement {class=FortControllerRequirement_HasOwnedTag, properties {ownedTag {tagName=Granted.Athena.BountyBoard.QuestActive}}}
    }
   }
  ]
 }
 Continue_C_9 {id=#14, class=Continue_C}
 ShowUserDataDrivenCannotUseReason_C_7 {
  id=#8
  class=ShowUserDataDrivenCannotUseReason_C
  outputs [#15]
  properties {
   cannotUseReason=BriefingAlreadyActive
   hoverChoiceSound=&JobBoard_ShadowBriefing_NotAvailable_Cue
   serviceBriefCollectionKeyOveride=Service.JobBoard.DisplayCannotUseReason
   serviceBriefConfigCollection=&JobBoardServiceBriefCollectionConfig
   softTaskIcon=&Cancel_Normal
  }
  sub_nodes [
   ConversationChoiceNode_9 {
    class=ConversationChoiceNode
    properties {defaultChoiceDisplayText=JOB ALREADY ACTIVE, choiceTags {gameplayTags [{tagName=Conversation.Choice.Category.Back}]}}
   }
   FortConversationRequirement_ControllerRequirement_12 {
    class=FortConversationRequirement_ControllerRequirement
    properties {
     nodeName=Require: Has GameplayTag
     requirement {class=FortControllerRequirement_HasOwnedTag, properties {ownedTag {tagName=Granted.Athena.JobBoardExperimental.QuestActive}}}
    }
   }
  ]
 }
 Continue_C_5 {id=#15, class=Continue_C}
 ConversationEntryPointNode_4 {
  id=#5
  class=ConversationEntryPointNode
  outputs [#16]
  properties {entryTag {tagName=Conversation.Entry.JobBoard.Bounty}}
 }
 Speech_C_13 {
  id=#16
  class=Speech_C
  outputs [
   #17
   #8
   #9
   #10
   #18
  ]
  properties {nodeName=""}
 }
 GrantPlayerBounty_C_1 {
  id=#18
  class=GrantPlayerBounty_C
  outputs []
  properties {bSharedQuestLimitsEnforced=false, bShouldOverrideSpeakerText=false, softTaskIcon=&T-Icon-Check-128}
  sub_nodes [
   ConversationChoiceNode_14 {class=ConversationChoiceNode, properties {defaultChoiceDisplayText=BOUNTY}}
   SideEffect_PlayServiceUsedEffects_C_3 {class=SideEffect_PlayServiceUsedEffects_C}
   SideEffect_Server_RecordEventToHistoryData_C_3 {class=SideEffect_Server_RecordEventToHistoryData_C, properties {event Type {tagName=Conversation.Event.GrantedQuest}}}
   SideEffect_Server_TrackJobBoardUser_C_5 {class=SideEffect_Server_TrackJobBoardUser_C}
   SideEffect_Server_JobBoardExperimental_DestroyOnUse_C_5 {class=SideEffect_Server_JobBoardBriefing_DeactivateOnUse_C}
   FortConversationRequirement_ControllerRequirement_35 {
    class=FortConversationRequirement_ControllerRequirement
    properties {
     nodeName=Require: Has Urgent Quest
     requirement {
      class=FortControllerRequirement_HasUrgentQuest
      properties {
       bInverseCondition=true
       bRequireExactMatch=false
       acceptedUrgentEventTags {gameplayTags [{tagName=SpecialEvent.HuntingPlayer.IsHunter}, {tagName=SpecialEvent.JobBoard.Experimental}]}
      }
     }
    }
   }
   FortConversationRequirement_ControllerRequirement_36 {
    class=FortConversationRequirement_ControllerRequirement
    properties {
     nodeName=Require: Has GameplayTag
     requirement {
      class=FortControllerRequirement_HasOwnedTag
      properties {bInverseCondition=true, ownedTag {tagName=Granted.Athena.JobBoardExperimental.QuestActive}}
     }
    }
   }
   FortConversationRequirement_ControllerRequirement_37 {
    class=FortConversationRequirement_ControllerRequirement
    properties {
     nodeName=Require: Has GameplayTag
     requirement {class=FortControllerRequirement_HasOwnedTag, properties {bInverseCondition=true, ownedTag {tagName=Granted.Athena.BountyBoard.QuestActive}}}
    }
   }
  ]
 }
 Continue_C_4 {
  id=#17
  class=Continue_C
  properties {softTaskIcon=&Cancel_Normal}
  sub_nodes [
   ConversationChoiceNode_5 {
    class=ConversationChoiceNode
    properties {defaultChoiceDisplayText=CANCEL, choiceTags {gameplayTags [{tagName=Conversation.Choice.CustomSortPriority.Highest}]}}
   }
   FortConversationRequirement_ControllerRequirement_5 {
    class=FortConversationRequirement_ControllerRequirement
    properties {
     nodeName=Require: Has GameplayTag
     requirement {
      class=FortControllerRequirement_HasOwnedTag
      properties {bInverseCondition=true, ownedTag {tagName=Granted.Athena.JobBoardExperimental.QuestActive}}
     }
    }
   }
   FortConversationRequirement_ControllerRequirement_7 {
    class=FortConversationRequirement_ControllerRequirement
    properties {
     nodeName=Require: Has GameplayTag
     requirement {class=FortControllerRequirement_HasOwnedTag, properties {bInverseCondition=true, ownedTag {tagName=Granted.Athena.BountyBoard.QuestActive}}}
    }
   }
  ]
 }
 ConversationEntryPointNode_3 {
  id=#4
  class=ConversationEntryPointNode
  outputs [#19]
  properties {entryTag {tagName=Conversation.Entry.JobBoard.VehicleDrop}}
 }
 Speech_C_7 {
  id=#19
  class=Speech_C
  outputs [
   #8
   #9
   #10
   #20
   #21
  ]
  properties {nodeName=""}
 }
 GrantStaticQuest_C_2 {
  id=#21
  class=GrantStaticQuest_C
  outputs []
  properties {
   bSharedQuestLimitsEnforced=false
   bShouldOverrideSpeakerText=false
   briefWidgetClassOverride=/JobBoard/Widgets/FortVehicleDropJobBrief.FortVehicleDropJobBrief_C
   questToGrantPtr=/JobBoard/QuestItems/quest_evergreen_jobboard_vehicle
   softTaskIcon=&T-Icon-Check-128
  }
  sub_nodes [
   FortConversationRequirement_ControllerRequirement_75 {
    class=FortConversationRequirement_ControllerRequirement
    properties {
     nodeName=Require: Has Urgent Quest
     requirement {
      class=FortControllerRequirement_HasUrgentQuest
      properties {
       bInverseCondition=true
       bRequireExactMatch=false
       acceptedUrgentEventTags {gameplayTags [{tagName=SpecialEvent.HuntingPlayer.IsHunter}, {tagName=SpecialEvent.JobBoard.Experimental}]}
      }
     }
    }
   }
   FortConversationRequirement_ControllerRequirement_76 {
    class=FortConversationRequirement_ControllerRequirement
    properties {
     nodeName=Require: Has GameplayTag
     requirement {
      class=FortControllerRequirement_HasOwnedTag
      properties {bInverseCondition=true, ownedTag {tagName=Granted.Athena.JobBoardExperimental.QuestActive}}
     }
    }
   }
   FortConversationRequirement_ControllerRequirement_77 {
    class=FortConversationRequirement_ControllerRequirement
    properties {
     nodeName=Require: Has GameplayTag
     requirement {class=FortControllerRequirement_HasOwnedTag, properties {bInverseCondition=true, ownedTag {tagName=Granted.Athena.BountyBoard.QuestActive}}}
    }
   }
   FortConversationRequirement_ControllerRequirement_78 {
    class=FortConversationRequirement_ControllerRequirement
    properties {
     nodeName=Require: Has GameplayTag
     requirement {class=FortControllerRequirement_HasOwnedTag, properties {bInverseCondition=true, ownedTag {tagName=Granted.Athena.UrgentQuest.Active}}}
    }
   }
   SideEffect_PlayServiceUsedEffects_C_15 {class=SideEffect_PlayServiceUsedEffects_C}
   SideEffect_Server_RecordEventToHistoryData_C_15 {class=SideEffect_Server_RecordEventToHistoryData_C, properties {event Type {tagName=Conversation.Event.GrantedQuest}}}
   SideEffect_Server_TrackJobBoardUser_C_15 {class=SideEffect_Server_TrackJobBoardUser_C}
   SideEffect_Server_JobBoardBriefing_SpawnActorOfClass_C_5 {class=SideEffect_Server_JobBoardBriefing_SpawnActorOfClass_C}
   SideEffect_Server_JobBoardBriefing_ApplyInstigatorGE_C_5 {class=SideEffect_Server_JobBoardBriefing_ApplyInstigatorGE_C}
   SideEffect_Server_JobBoardBriefing_DeactivateOnUse_C_5 {class=SideEffect_Server_JobBoardBriefing_DeactivateOnUse_C}
   ConversationChoiceNode_30 {class=ConversationChoiceNode, properties {defaultChoiceDisplayText=VEHICLE BRIEFING}}
  ]
 }
 Continue_C_3 {
  id=#20
  class=Continue_C
  properties {softTaskIcon=&Cancel_Normal}
  sub_nodes [
   ConversationChoiceNode_6 {
    class=ConversationChoiceNode
    properties {defaultChoiceDisplayText=CANCEL, choiceTags {gameplayTags [{tagName=Conversation.Choice.CustomSortPriority.Highest}]}}
   }
   FortConversationRequirement_ControllerRequirement_22 {
    class=FortConversationRequirement_ControllerRequirement
    properties {
     nodeName=Require: Has GameplayTag
     requirement {
      class=FortControllerRequirement_HasOwnedTag
      properties {bInverseCondition=true, ownedTag {tagName=Granted.Athena.JobBoardExperimental.QuestActive}}
     }
    }
   }
   FortConversationRequirement_ControllerRequirement_8 {
    class=FortConversationRequirement_ControllerRequirement
    properties {
     nodeName=Require: Has GameplayTag
     requirement {class=FortControllerRequirement_HasOwnedTag, properties {bInverseCondition=true, ownedTag {tagName=Granted.Athena.BountyBoard.QuestActive}}}
    }
   }
   FortConversationRequirement_ControllerRequirement_41 {
    class=FortConversationRequirement_ControllerRequirement
    properties {
     nodeName=Require: Has GameplayTag
     requirement {class=FortControllerRequirement_HasOwnedTag, properties {bInverseCondition=true, ownedTag {tagName=Granted.Athena.UrgentQuest.Active}}}
    }
   }
  ]
 }
 ConversationEntryPointNode_2 {
  id=#3
  class=ConversationEntryPointNode
  outputs [#22]
  properties {entryTag {tagName=Conversation.Entry.JobBoard.SupplyDropGeneric}}
 }
 Speech_C_5 {
  id=#22
  class=Speech_C
  outputs [
   #8
   #9
   #10
   #23
   #24
  ]
  properties {nodeName=""}
 }
 GrantStaticQuest_C_1 {
  id=#24
  class=GrantStaticQuest_C
  outputs []
  properties {
   bSharedQuestLimitsEnforced=false
   bShouldOverrideSpeakerText=false
   briefWidgetClassOverride=/JobBoard/Widgets/FortSupplyDropJobBrief.FortSupplyDropJobBrief_C
   questToGrantPtr=/JobBoard/QuestItems/quest_evergreen_jobboard_supplydrop
   softTaskIcon=&T-Icon-Check-128
  }
  sub_nodes [
   SideEffect_PlayServiceUsedEffects_C_12 {class=SideEffect_PlayServiceUsedEffects_C}
   SideEffect_Server_RecordEventToHistoryData_C_12 {class=SideEffect_Server_RecordEventToHistoryData_C, properties {event Type {tagName=Conversation.Event.GrantedQuest}}}
   SideEffect_Server_TrackJobBoardUser_C_12 {class=SideEffect_Server_TrackJobBoardUser_C}
   SideEffect_Server_JobBoardBriefing_SpawnActorOfClass_C_3 {
    class=SideEffect_Server_JobBoardBriefing_SpawnActorOfClass_C
    properties {spawnerClass=/JobBoard/Blueprints/Shared/B_Athena_JobBoardBriefing_ActorSpawner_SupplyDrop.B_Athena_JobBoardBriefing_ActorSpawner_SupplyDrop_C}
   }
   SideEffect_Server_JobBoardBriefing_ApplyInstigatorGE_C_3 {class=SideEffect_Server_JobBoardBriefing_ApplyInstigatorGE_C}
   SideEffect_Server_JobBoardBriefing_DeactivateOnUse_C_3 {class=SideEffect_Server_JobBoardBriefing_DeactivateOnUse_C}
   FortConversationRequirement_ControllerRequirement_62 {
    class=FortConversationRequirement_ControllerRequirement
    properties {
     nodeName=Require: Has Recorded Conversation Event With NPC
     requirement {
      class=FortControllerRequirement_HasRecordedConversationEventWithNPC
      properties {bInverseCondition=true, queryMethod=CheckAgainstCurrentConversationParticipant, queryEventTag {tagName=Conversation.Event.GrantedQuest}}
     }
    }
   }
   FortConversationRequirement_ControllerRequirement_63 {
    class=FortConversationRequirement_ControllerRequirement
    properties {
     nodeName=Require: Has Urgent Quest
     requirement {
      class=FortControllerRequirement_HasUrgentQuest
      properties {
       bInverseCondition=true
       bRequireExactMatch=false
       acceptedUrgentEventTags {gameplayTags [{tagName=SpecialEvent.HuntingPlayer.IsHunter}, {tagName=SpecialEvent.JobBoard.Experimental}]}
      }
     }
    }
   }
   FortConversationRequirement_ControllerRequirement_64 {
    class=FortConversationRequirement_ControllerRequirement
    properties {
     nodeName=Require: Has GameplayTag
     requirement {
      class=FortControllerRequirement_HasOwnedTag
      properties {bInverseCondition=true, ownedTag {tagName=Granted.Athena.JobBoardExperimental.QuestActive}}
     }
    }
   }
   FortConversationRequirement_ControllerRequirement_65 {
    class=FortConversationRequirement_ControllerRequirement
    properties {
     nodeName=Require: Has GameplayTag
     requirement {class=FortControllerRequirement_HasOwnedTag, properties {bInverseCondition=true, ownedTag {tagName=Granted.Athena.BountyBoard.QuestActive}}}
    }
   }
   FortConversationRequirement_ControllerRequirement_66 {
    class=FortConversationRequirement_ControllerRequirement
    properties {
     nodeName=Require: Has GameplayTag
     requirement {class=FortControllerRequirement_HasOwnedTag, properties {bInverseCondition=true, ownedTag {tagName=Granted.Athena.UrgentQuest.Active}}}
    }
   }
   ConversationChoiceNode_27 {class=ConversationChoiceNode, properties {defaultChoiceDisplayText=SUPPLY DROP QUEST}}
  ]
 }
 Continue_C_2 {
  id=#23
  class=Continue_C
  properties {softTaskIcon=&Cancel_Normal}
  sub_nodes [
   ConversationChoiceNode_10 {
    class=ConversationChoiceNode
    properties {defaultChoiceDisplayText=CANCEL, choiceTags {gameplayTags [{tagName=Conversation.Choice.CustomSortPriority.Highest}]}}
   }
   FortConversationRequirement_ControllerRequirement_23 {
    class=FortConversationRequirement_ControllerRequirement
    properties {
     nodeName=Require: Has GameplayTag
     requirement {
      class=FortControllerRequirement_HasOwnedTag
      properties {bInverseCondition=true, ownedTag {tagName=Granted.Athena.JobBoardExperimental.QuestActive}}
     }
    }
   }
   FortConversationRequirement_ControllerRequirement_9 {
    class=FortConversationRequirement_ControllerRequirement
    properties {
     nodeName=Require: Has GameplayTag
     requirement {class=FortControllerRequirement_HasOwnedTag, properties {bInverseCondition=true, ownedTag {tagName=Granted.Athena.BountyBoard.QuestActive}}}
    }
   }
   FortConversationRequirement_ControllerRequirement_42 {
    class=FortConversationRequirement_ControllerRequirement
    properties {
     nodeName=Require: Has GameplayTag
     requirement {class=FortControllerRequirement_HasOwnedTag, properties {bInverseCondition=true, ownedTag {tagName=Granted.Athena.UrgentQuest.Active}}}
    }
   }
  ]
 }
 ConversationEntryPointNode_1 {
  id=#2
  class=ConversationEntryPointNode
  outputs [#25]
  properties {entryTag {tagName=Conversation.Entry.JobBoard.TreasureHunt}}
 }
 Speech_C_2 {
  id=#25
  class=Speech_C
  outputs [
   #8
   #9
   #10
   #26
   #27
  ]
  properties {nodeName=""}
 }
 GrantStaticQuest_C_0 {
  id=#27
  class=GrantStaticQuest_C
  outputs []
  properties {
   bSharedQuestLimitsEnforced=false
   bShouldOverrideSpeakerText=false
   briefWidgetClassOverride=/JobBoard/Widgets/FortTreasureHuntingJobBrief.FortTreasureHuntingJobBrief_C
   questToGrantPtr=/JobBoard/QuestItems/quest_evergreen_jobboard_treasurehunt
   softTaskIcon=&T-Icon-Check-128
  }
  sub_nodes [
   FortConversationRequirement_ControllerRequirement_49 {
    class=FortConversationRequirement_ControllerRequirement
    properties {
     nodeName=Require: Has Urgent Quest
     requirement {
      class=FortControllerRequirement_HasUrgentQuest
      properties {
       bInverseCondition=true
       bRequireExactMatch=false
       acceptedUrgentEventTags {gameplayTags [{tagName=SpecialEvent.HuntingPlayer.IsHunter}, {tagName=SpecialEvent.JobBoard.Experimental}]}
      }
     }
    }
   }
   FortConversationRequirement_ControllerRequirement_50 {
    class=FortConversationRequirement_ControllerRequirement
    properties {
     nodeName=Require: Has GameplayTag
     requirement {
      class=FortControllerRequirement_HasOwnedTag
      properties {bInverseCondition=true, ownedTag {tagName=Granted.Athena.JobBoardExperimental.QuestActive}}
     }
    }
   }
   FortConversationRequirement_ControllerRequirement_55 {
    class=FortConversationRequirement_ControllerRequirement
    properties {
     nodeName=Require: Has GameplayTag
     requirement {class=FortControllerRequirement_HasOwnedTag, properties {bInverseCondition=true, ownedTag {tagName=Granted.Athena.BountyBoard.QuestActive}}}
    }
   }
   FortConversationRequirement_ControllerRequirement_56 {
    class=FortConversationRequirement_ControllerRequirement
    properties {
     nodeName=Require: Has GameplayTag
     requirement {class=FortControllerRequirement_HasOwnedTag, properties {bInverseCondition=true, ownedTag {tagName=Granted.Athena.UrgentQuest.Active}}}
    }
   }
   ConversationChoiceNode_23 {class=ConversationChoiceNode, properties {defaultChoiceDisplayText=JOB BOARD BRIEFING}}
   SideEffect_PlayServiceUsedEffects_C_10 {class=SideEffect_PlayServiceUsedEffects_C}
   SideEffect_Server_RecordEventToHistoryData_C_10 {class=SideEffect_Server_RecordEventToHistoryData_C, properties {event Type {tagName=Conversation.Event.GrantedQuest}}}
   SideEffect_Server_TrackJobBoardUser_C_10 {class=SideEffect_Server_TrackJobBoardUser_C}
   SideEffect_Server_JobBoardBriefing_SpawnActorOfClass_C_1 {
    class=SideEffect_Server_JobBoardBriefing_SpawnActorOfClass_C
    properties {spawnerClass=/JobBoard/Blueprints/Shared/B_Athena_JobBoardBriefing_ActorSpawner_Plunder.B_Athena_JobBoardBriefing_ActorSpawner_Plunder_C}
   }
   SideEffect_Server_JobBoardBriefing_ApplyInstigatorGE_C_1 {class=SideEffect_Server_JobBoardBriefing_ApplyInstigatorGE_C}
   SideEffect_Server_JobBoardBriefing_DeactivateOnUse_C_1 {class=SideEffect_Server_JobBoardBriefing_DeactivateOnUse_C}
  ]
 }
 Continue_C_0 {
  id=#26
  class=Continue_C
  properties {softTaskIcon=&Cancel_Normal}
  sub_nodes [
   ConversationChoiceNode_8 {
    class=ConversationChoiceNode
    properties {defaultChoiceDisplayText=CANCEL, choiceTags {gameplayTags [{tagName=Conversation.Choice.CustomSortPriority.Highest}]}}
   }
   FortConversationRequirement_ControllerRequirement_24 {
    class=FortConversationRequirement_ControllerRequirement
    properties {
     nodeName=Require: Has GameplayTag
     requirement {
      class=FortControllerRequirement_HasOwnedTag
      properties {bInverseCondition=true, ownedTag {tagName=Granted.Athena.JobBoardExperimental.QuestActive}}
     }
    }
   }
   FortConversationRequirement_ControllerRequirement_10 {
    class=FortConversationRequirement_ControllerRequirement
    properties {
     nodeName=Require: Has GameplayTag
     requirement {class=FortControllerRequirement_HasOwnedTag, properties {bInverseCondition=true, ownedTag {tagName=Granted.Athena.BountyBoard.QuestActive}}}
    }
   }
   FortConversationRequirement_ControllerRequirement_43 {
    class=FortConversationRequirement_ControllerRequirement
    properties {
     nodeName=Require: Has GameplayTag
     requirement {class=FortControllerRequirement_HasOwnedTag, properties {bInverseCondition=true, ownedTag {tagName=Granted.Athena.UrgentQuest.Active}}}
    }
   }
  ]
 }
 ConversationEntryPointNode_7 {
  id=#1
  class=ConversationEntryPointNode
  outputs [#28]
  properties {entryTag {tagName=Conversation.Entry.JobBoard}}
 }
 Speech_C_11 {
  id=#28
  class=Speech_C
  outputs [#8, #9, #10]
  properties {nodeName="", generalConfig {defaultMessage="NO JOBS AVAILABLE!"}}
 }
]
