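; Conversation requirement for job board briefings that checks whether loot can be given for a configured loot tier group.
; It gets the participant NPC from the conversation context and returns FailedAndHidden if that actor is invalid.
; It then trial-picks drops at the current loot level, passing if any result, otherwise FailedButVisible.
; An inverse flag flips this to FailedAndHidden when loot exists and Passed when none does.
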(fn IsRequirementSatisfied
  (params (Context :type Conversation%20Context%20Structure%20%28by%20ref%29))
  (bind _binversecondition (Variables|Default|GetbInverseCondition))
  ;[0 0 0] Check if loot exists for the tier drop indicated and that stock qty is not 0
  (bind _returnvalue (Class|FortConversationContextLibrary|GetParticipantActorNPC Context))
  (Utilities|IsValid _returnvalue
    (:Is%20Valid
      (if true
        (bind _outloottodrop (Loot|PickLootDrops (Variables|Default|GetLootTierGroupName) (Loot|GetLootLevel) -1 "(GameplayTags=)"))
        (if (> (Utilities|Array|Length _outloottodrop) 0)
          (if _binversecondition
            (return "FailedAndHidden")
            (else
              (return "Passed")))
          (elif _binversecondition
            (return "Passed")
            (else
              (return "FailedButVisible"))))
        (else
          (return "FailedAndHidden"))))
    (:Is%20Not%20Valid
      (return "FailedAndHidden"))))
