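; Server-side conversation side effect for the Job Board briefing that applies a gameplay effect to the participating player.
; ServerCauseSideEffect gets the participant's player controller from the conversation context and, if valid, finds the Fort ability system component on that player's pawn.
; It applies the configurable InstigatorGameplayEffect class to that component at level zero, and does nothing when the controller is invalid.
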(fn ServerCauseSideEffect
  (params (Context :type Conversation%20Context%20Structure%20%28by%20ref%29))
  (bind _returnvalue (Class|FortConversationContextLibrary|GetParticipantPlayerController Context))
  (Utilities|IsValid _returnvalue
    (:Is%20Valid
      (GameplayEffects|ApplyGameplayEffectToSelf :self (Ability|GetFortAbilitySystemComponent (Game|Player|GetPlayerPawn _returnvalue)) :GameplayEffectClass (Variables|Default|GetInstigatorGameplayEffect) :Level 0.0))
    (:Is%20Not%20Valid
      _))
  (return))
