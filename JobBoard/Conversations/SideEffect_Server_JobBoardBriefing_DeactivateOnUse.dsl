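; Server-side conversation side effect for the Job Board that deactivates a briefing prop when it is used.
; It gets the NPC participant actor from the conversation context and casts it to the briefing station prop parent Blueprint.
; After confirming that the prop's non-player conversation component is valid, it calls DeactivateonUse on the prop, and does nothing if any check or the cast fails.
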(fn ServerCauseSideEffect
  (params (Context :type Conversation%20Context%20Structure%20%28by%20ref%29))
  ;[0 0 0] If actor and conversation component are valid, deactivate the prop on use
  (bind _returnvalue (Class|FortConversationContextLibrary|GetParticipantActorNPC Context))
  (Utilities|IsValid _returnvalue
    (:Is%20Valid
      (bind _asbp_station_prop_job_board_briefing_parent (Utilities|Casting|CastToBP_StationProp_JobBoard_Briefing_Parent _returnvalue)
        (:then
          (Utilities|IsValid (Variables|/Game/Building/ActorBlueprints/Stations/BP_StationProp_Parent.BP_StationProp_Parent_C|GetNonPlayerConversationComponent _asbp_station_prop_job_board_briefing_parent)
            (:Is%20Valid
              (Class|BPStationPropJobBoardBriefingParent|DeactivateonUse _asbp_station_prop_job_board_briefing_parent)
              (return))
            (:Is%20Not%20Valid
              _)))
        (:CastFailed
          _))
      (return))
    (:Is%20Not%20Valid
      _)))
