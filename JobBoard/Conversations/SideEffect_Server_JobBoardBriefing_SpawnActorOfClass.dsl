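; Server-side conversation side effect for a job board briefing that spawns an actor of a configurable class above the conversation's NPC participant, presumably the board.
; It validates the NPC and player participants, then copies the NPC's transform with Z raised by a scalable-float height offset.
; The actor always spawns regardless of collision, with the player participant set as its owner.
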(fn ServerCauseSideEffect
  (params (Context :type Conversation%20Context%20Structure%20%28by%20ref%29))
  ;[0 0 0] Spawn the actor at the height above the board specified by data table and set the participant as the owner
  (bind _returnvalue (Class|FortConversationContextLibrary|GetParticipantActorNPC Context))
  (Utilities|IsValid _returnvalue
    (:Is%20Valid
      (bind _returnvalue_2 (Class|FortConversationContextLibrary|GetParticipantActorPlayer Context))
      (Utilities|IsValid _returnvalue_2
        (:Is%20Valid
          (bind (_returnvalue_location _returnvalue_rotation _returnvalue_scale) (Math|Transform|BreakTransform (Transformation|GetActorTransform _returnvalue)))
          (Game|SpawnActorfromClass (Variables|Default|GetSpawnerClass) (Math|Transform|MakeTransform (Math|Vector|MakeVector (.x (Transformation|GetActorLocation _returnvalue)) (.y (Transformation|GetActorLocation _returnvalue)) (+ (.z (Transformation|GetActorLocation _returnvalue)) (ScalableFloat|GetValueatLevel (Variables|Default|GetHeightAboveGround) 1.0))) _returnvalue_rotation _returnvalue_scale) "AlwaysSpawn" "MultiplyWithRoot" _returnvalue_2 0)
          (return))))))
