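; Server-side conversation side effect for Job Board Briefing station props that grants loot straight into the talking player's inventory.
; It validates the player controller and station prop from the conversation context, then picks drops from a configurable loot tier group.
; Ammo drops keep count and level; other drops are assumed weapons, given with a random mod set if a scalable-float toggle is on.
; A successful loot pick deactivates the station prop.
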(fn AddAmmoToInstigatorInventory
  (params (InstigatingPlayerController :type Fort%20Player%20Controller%20Object%20Reference) (AmmoItem :type Fort%20Ammo%20Item%20Definition%20Object%20Reference) (Count :type Integer) (Level :type Integer))
  (Inventory|GiveItemtoInventoryOwner :InventoryOwner InstigatingPlayerController :ItemDefinition AmmoItem :NumberToGive Count :bNotifyPlayer true :ItemLevel Level :PickupInstigatorHandle -1 :bUseItemPickupAnalyticEvent true :WeaponAmmoOverride -1))

(fn AddWeapontoInstigatorInventory
  (params (WeaponItemDef :type Fort%20Weapon%20Item%20Definition%20Object%20Reference) (InstigatingPlayerController :type Fort%20Player%20Controller%20Object%20Reference))
  ;[0 0 0] Determine if weapon should come with mods. If not, add it as a standard item
  (if (ScalableFloat|GetValueasBool (Variables|Default|GetShouldApplyWeaponMod))
    (bind _returnvalue (Class|FortWeaponModFunctionLibrary|GetRandomWeaponModSetData WeaponItemDef))
    (Inventory|GiveWeaponwithModSettoInventoryOwner InstigatingPlayerController WeaponItemDef _returnvalue)
    (else
      (InventoryManagement|AddItem InstigatingPlayerController WeaponItemDef 1))))

(fn GetPlayerandStationPropReferences
  (params (Context :type Conversation%20Context%20Structure))
  ;[0 0 0] Helper function for getting a valid reference to the station prop and player controller
  (bind _returnvalue (Class|FortConversationContextLibrary|GetParticipantActorNPC Context))
  (Utilities|IsValid _returnvalue
    (:Is%20Valid
      (bind _asbp_station_prop_job_board_briefing_parent (Utilities|Casting|CastToBP_StationProp_JobBoard_Briefing_Parent _returnvalue)
        (:then
          (bind _returnvalue_2 (Class|FortConversationContextLibrary|GetParticipantActorPlayer Context))
          (Utilities|IsValid _returnvalue_2
            (:Is%20Valid
              (bind _asfort_player_controller (Utilities|Casting|CastToFortPlayerController _returnvalue_2)
                (:then
                  (return _asfort_player_controller _asbp_station_prop_job_board_briefing_parent true))
                (:CastFailed
                  _)))
            (:Is%20Not%20Valid
              _))
          (return 0 0 false))
        (:CastFailed
          _)))
    (:Is%20Not%20Valid
      _)))

(fn GiveLootDropsToPlayer
  (params (InstigatingPlayerController :type Fort%20Player%20Controller%20Object%20Reference))
  ;[0 0 0] Handle loot drops picked from the provided tier group name accordingly depending on if the drop is a weapon/ammo stack
  (bind (_outloottodrop _returnvalue) (Loot|PickLootDrops (Variables|Default|GetLootTierGroupName) (Loot|GetLootLevel) -1 "(GameplayTags=)"))
  (if _returnvalue
    (for _array_element _outloottodrop
      ;; Check if the drop is an ammo stack or a weapon - add the respective item(s) to inventory
      (bind _asfort_ammo_item_definition (Utilities|Casting|CastToFortAmmoItemDefinition (Class|FortItemFunctionLibrary|BreakFortItemEntry _array_element))
        (:then
          (CallFunction|AddAmmotoInstigatorInventory InstigatingPlayerController _asfort_ammo_item_definition (.Count (Class|FortItemFunctionLibrary|BreakFortItemEntry _array_element)) (.Level (Class|FortItemFunctionLibrary|BreakFortItemEntry _array_element))))
        (:CastFailed
          (bind _asfort_weapon_item_definition (Utilities|Casting|CastToFortWeaponItemDefinition (Class|FortItemFunctionLibrary|BreakFortItemEntry _array_element)))
          (CallFunction|AddWeaponToInstigatorInventory _asfort_weapon_item_definition InstigatingPlayerController))))
    (return true)
    (else
      (return false))))

(fn ServerCauseSideEffect
  (params (Context :type Conversation%20Context%20Structure%20%28by%20ref%29))
  ;[0 0 0] Give loot for the loot tier group provided to the instigating player - currently assumes weapons and ammo, not other items!
  (bind (_instigatingplayercontroller _stationprop _brefsvalid) (CallFunction|GetPlayerAndStationPropReferences Context))
  (if _brefsvalid
    (bind _blootdropsvalid (CallFunction|GiveLootDropstoPlayer _instigatingplayercontroller))
    (if _blootdropsvalid
      (Class|BPStationPropJobBoardBriefingParent|DeactivateonUse _stationprop)
      (return)
      (else
        (return)))
    (else
      (return))))
