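; Conversation side effect for the job board that registers the conversing player as a service user of the board, server-side.
; ServerCauseSideEffect takes the NPC participant from the conversation context and casts it to BP_StationProp_JobBoard.
; On a valid board, it calls TrackServiceUser with the participant's player controller; invalid participants or failed casts do nothing.
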(fn ServerCauseSideEffect
  (params (Context :type Conversation%20Context%20Structure%20%28by%20ref%29))
  (bind _returnvalue (Class|FortConversationContextLibrary|GetParticipantActorNPC Context))
  (Utilities|IsValid _returnvalue
    (:Is%20Valid
      (bind _asbp_station_prop_job_board (Utilities|Casting|CastToBP_StationProp_JobBoard _returnvalue)
        (:then
          (Utilities|IsValid _asbp_station_prop_job_board
            (:Is%20Valid
              (bind _returnvalue_2 (Class|FortConversationContextLibrary|GetParticipantPlayerController Context))
              (Class|BPStationPropJobBoard|TrackServiceUser _asbp_station_prop_job_board _returnvalue_2))
            (:Is%20Not%20Valid
              _))
          (return))
        (:CastFailed
          _)))
    (:Is%20Not%20Valid
      _)))
